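# Translucent UI material for the background image of a billboard row in the Discovery browser.
# A Background Image texture parameter is sampled with clamped UVs, scaled to fill the widget around a pivot.
# Separate branches build a rounded-corner mask, a powered edge falloff, and an edge fade from mirrored UVs, eased vertically and smoothstepped over a pixel-sized horizontal band.
# The graph shows no links joining these branches to the material outputs.
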Material /DiscoveryBrowser/Rows/Billboard/Materials/M_UI_Discovery_Row_Billboard_Background
graph {
 properties {
  bAutomaticallySetUsageInEditor=false
  bUsedWithStaticMesh=false
  blendMode=BLEND_Translucent
  materialDomain=MD_UI
  usedShadingModels=Default Lit
 }
 nodes [
  n1 = MaterialFunctionCall() {materialFunction=&GetUserInterfaceUV}
  n2 = Min(A=@n1.Normalized UV, B=OneMinus(None=@n1.Normalized UV))
  n3 = Power(Base=OneMinus(None=Multiply(A=@n2) {constB=2})) {constExponent=8}
  LinearInterpolate(A=ComponentMask(None=@n3) {r=true}, Alpha=ComponentMask(None=@n3) {g=true})
  MaterialFunctionCall(Corner Radius=Constant() {r=10}, Radius in Pixels=Constant() {r=1}) {materialFunction=&MF_UI_CoreUI_RoundedMask}
  n4 = MaterialFunctionCall() {materialFunction=&GetUserInterfaceUV}
  n5 = Multiply(A=Min(A=@n4.Normalized UV, B=OneMinus(None=@n4.Normalized UV))) {constB=2}
  n6 = ComponentMask(None=@n5) {r=true}
  n7 = Constant()
  n8 = MaterialFunctionCall!(Value PX=Constant() {r=64}, UV Pixel Size, X or Y) {materialFunction=/FortUILibrary/MaterialFunctions/Basics/MF_UI_PixelsToNormalized}
  Multiply(
   A=MaterialFunctionCall("Value (0-1)"=ComponentMask(None=@n5) {g=true}) {materialFunction=&MF_UI_EaseCurves}.EaseOutQuint
   B=SmoothStep(Value=MaterialFunctionCall(Input=@n6, Input Low=@n7, Input High=@n8."Normalized (>1px)") {materialFunction=&RemapValue_0-1}.Result)
  )
  n9 = TextureObjectParameter() {parameterName=Background Image, texture=/MeshModelingToolsetExp/Textures/DefaultTexture}
  TextureSample(
   UVs=MaterialFunctionCall(
    UV=MaterialFunctionCall() {materialFunction=&GetUserInterfaceUV}.Normalized UV
    TextureSize=TextureProperty(Tex=@n9)
    ScalePivotPoint=Constant2Vector() {r=0.5}
   ) {materialFunction=&MF_UI_UVScaling}.Scaled To Fill
   Tex=@n9
  ) {samplerSource=SSM_Clamp_WorldGroupSettings}
 ]
}
