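# UI material for the foreground image of a Discovery browser billboard row, adding a parallax effect on hover or focus.
# Cursor position, clamped and remapped against widget bounds derived from screen-position parameters, yields a ratio; focus substitutes a looping time-based triangle wave.
# That ratio drives horizontal pixel offset, alongside a slight 3D rotation of the fitted texture.
# Alpha gets edge-pixel clipping only in the upper half.
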Material /DiscoveryBrowser/Rows/Billboard/Materials/M_UI_Discovery_Row_Billboard_Foreground
graph {
 properties {
  bAutomaticallySetUsageInEditor=false
  bUsedWithStaticMesh=false
  blendMode=BLEND_Translucent
  materialDomain=MD_UI
  usedShadingModels=Default Lit
 }
 nodes [
  n1 = MaterialFunctionCall() {materialFunction=&GetUserInterfaceUV}
  n2 = Constant()
  n3 = NamedRerouteUsage()
  n4 = Subtract(A=@n2, B=@n3)
  n5 = Add(A=@n2, B=@n3)
  n6 = NamedRerouteDeclaration(None=NamedRerouteUsage()) {name=Widget Top Left on Screen, nodeColor=<rgba 0 0 0 1>}
  n7 = MaterialFunctionCall() {materialFunction=&GetUserInterfaceUV}
  n8 = Clamp(
   None=MaterialFunctionCall() {materialFunction=/Engine/Functions/UserInterface/MF_UI_CursorPosition}.Result
   Min=@n6
   Max=NamedRerouteDeclaration(None=Add(A=@n6, B=@n7.Pixel Size)) {name=Widget Bottom Right On Screen, nodeColor=<rgba 0 0 0 1>}
  )
  n9 = MaterialFunctionCall(
   Input=ComponentMask(None=NamedRerouteDeclaration(None=@n8) {name=Pixel Range, nodeColor=<rgba 0 0 0 1>}) {r=true}
   Input Low=ComponentMask(None=NamedRerouteUsage()) {r=true}
   Input High=ComponentMask(None=NamedRerouteUsage()) {r=true}
  ) {desc=Set Range for Mouse Position, materialFunction=&RemapValue_0-1}
  n10 = Frac(None=Multiply(A=Time()) {constB=0.25})
  n11 = OneMinus(None=@n10)
  n12 = NamedRerouteDeclaration(None=Multiply(A=Min(A=@n10, B=@n11)) {constB=2}) {name=Focused Time Animation, nodeColor=<rgba 0 0 0 1>}
  n13 = ScalarParameter() {parameterName=Is Focused}
  n14 = NamedRerouteDeclaration(None=SmoothStep(Value=LinearInterpolate(A=@n9.Result, B=@n12, Alpha=@n13) {constAlpha=1})) {name=Mouse Position Ratio, nodeColor=<rgba 0 0 0 1>}
  n15 = AppendVector(
   A=MaterialFunctionCall!(Value PX=LinearInterpolate(A=@n4, B=@n5, Alpha=@n14) {desc=Padding Offset}, UV Pixel Size, X or Y) {materialFunction=/FortUILibrary/MaterialFunctions/Basics/MF_UI_PixelsToNormalized}.Normalized
   B=Constant()
  )
  n16 = TextureObjectParameter() {parameterName=Foreground Image, texture=/Game/Global/Textures/Default/Blanks/T_Blank_White_ZeroAlpha_sRGBColor}
  n17 = Constant()
  n18 = LinearInterpolate(
   A=ScalarParameter() {defaultValue=0.01, parameterName=Parallax Rotation Left}
   B=Multiply(A=ScalarParameter() {defaultValue=0.0025, parameterName=Parallax Rotation Right}) {constB=-1}
   Alpha=LinearInterpolate(A=NamedRerouteUsage(), B=NamedRerouteUsage(), Alpha=ScalarParameter() {parameterName=Is Focused}) {constAlpha=1}
  ) {constA=-0.1, constB=0.1}
  n19 = LinearInterpolate(A=ScalarParameter() {parameterName=Is Hovered}, Alpha=ScalarParameter() {parameterName=Is Focused}) {constAlpha=1}
  n20 = MaterialFunctionCall!(
   UV=MaterialFunctionCall(UV=Add(A=@n1.Normalized UV, B=@n15), TextureSize=TextureProperty(Tex=@n16), ScalePivotPoint=Constant2Vector() {g=1, r=1}) {materialFunction=&MF_UI_UVScaling}.Scaled to Fit
   Angle=MaterialFunctionCall(R=@n17, G=LinearInterpolate(B=@n18, Alpha=@n19), B=@n17) {materialFunction=&AppendMany}.RGB
   Camera FOV=Constant() {r=45}
   Camera Offset=Constant3Vector() {constant=<rgba 0 0 2.41 1>}
   Offset from Pivot=Constant()
   Texture Scale
   Texture Size
  ) {materialFunction=/FortUILibrary/MaterialFunctions/Basics/MF_UI_Rotate3D}
  n21 = TextureSample(UVs=@n20.Rotated UV, Tex=@n16) {samplerSource=SSM_Clamp_WorldGroupSettings}
  n22 = MaterialFunctionCall() {materialFunction=&GetUserInterfaceUV}
  LinearInterpolate(
   A=MaterialFunctionCall!(UVs=@n20.Rotated UV, Alpha=@n21.A) {materialFunction=/FortUILibrary/MaterialFunctions/Basics/MF_UI_ClipEdgePixel}.Result
   B=@n21.A
   Alpha=Round(None=ComponentMask(None=@n22.Normalized UV) {g=true})
  )
  MaterialFunctionCall() {materialFunction=/FortUILibrary/MaterialFunctions/Basics/MF_UI_ViewportScale}
  n23 = AppendVector(A=ScalarParameter() {parameterName=ScreenPosX}, B=ScalarParameter() {parameterName=ScreenPosY})
  n24 = ComponentMask(None=ViewSize()) {g=true}
  n25 = Constant() {r=1080}
  NamedRerouteDeclaration(None=Multiply(A=Abs(None=@n23), B=Divide(A=@n24, B=@n25))) {name=Widget Screen Position, nodeColor=<rgba 0 0 0 1>}
  n26 = ScalarParameter() {parameterName=Is Hovered}
  n27 = ScalarParameter() {parameterName=Is Focused}
  NamedRerouteDeclaration(None=LinearInterpolate(B=ScalarParameter() {defaultValue=2, parameterName=Parallax Offset}, Alpha=LinearInterpolate(A=@n26, Alpha=@n27) {constAlpha=1})) {name=Parallax Offset Pixel Amount, nodeColor=<rgba 0 0 0 1>}
 ]
}
